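; Thrown-shield projectile that removes the shield item from the thrower's inventory and turns it into a deployed shield on impact.
; On hit, it ignores deployed shields; otherwise the server checks energy and target, then spawns the shield, hands over energy depletion and item entry, and attaches or vehicle-aligns it.
; Failed deploys and water entry drop a pickup instead, while remote clients play a local impact GameplayCue.
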(event EventBeginPlay
  (Components|Movement|SetUpdatedComponent (Variables|Default|GetRotatingMovement) (Variables|Default|GetSkeletalMesh))
  (CallFunction|RemoveShieldItem))

(event Collision|EventHit
  (params (MyComp :type Primitive%20Component%20Object%20Reference) (Other :type Actor%20Object%20Reference) (OtherComp :type Primitive%20Component%20Object%20Reference) (bSelfMoved :type Boolean) (HitLocation :type Vector :default "0, 0, 0") (HitNormal :type Vector :default "0, 0, 0") (NormalImpulse :type Vector :default "0, 0, 0") (Hit :type Hit%20Result%20Structure%20%28by%20ref%29))
  (bind (_bblockinghit _binitialoverlap _time _distance _location _impactpoint _normal) (Collision|BreakHitResult Hit))
  (bind _returnvalue (CallFunction|ShouldBounceOff Other))
  (if (not _returnvalue)
    (Networking|SwitchHasAuthority
      (:Authority
        (if (> (CallFunction|GetEnergy) 0)
          (bind _returnvalue_2 (CallFunction|CanDeployonActor Other OtherComp))
          (if _returnvalue_2
            ;[1 1 1] If we have enough energy, proceed with spawning the deployed shield
            (bind _outnormal (CallFunction|ClampNormal _normal Other))
            (bind _returnvalue_3 (CallFunction|SpawnBGA _outnormal Other OtherComp _location false Hit))
            (bind _asbga_shielder_deployed (Utilities|Casting|CastToBGA_Shielder_Deployed _returnvalue_3))
            (CallFunction|TransferEnergy _asbga_shielder_deployed)
            (CallFunction|GiveItemEntrytoDeployedShield _asbga_shielder_deployed)
            (CallFunction|TryAttach Other _asbga_shielder_deployed)
            (CallFunction|MatchRotationforVehicles Other _asbga_shielder_deployed)
            (Actor|DestroyActor)
            (else
              _)))
        ;[1 1 1] If we don't want to deploy the shield for whatever reason, just spawn the item pickup
        (CallFunction|SpawnItemPickup)
        (Actor|DestroyActor))
      (:Remote
        (GameplayCue|ExecuteGameplayCue%28Burst%29Local self (Variables|Default|GetGameplay%20Cue%20Tag) (GameplayCue|MakeGameplayCueParametersfromHitResult Hit))))))

(event OnEnterWaterEvent(WaterInteractionComponent)
  (params (WaterBody :type Fort%20Water%20Body%20Actor%20Object%20Reference) (WaterInteractionComponent :type Fort%20Water%20Interaction%20Component%20Object%20Reference) (bIsFirstBody :type Boolean))
  (Networking|SwitchHasAuthority
    (:Authority
      (CallFunction|SpawnItemPickup)
      (Actor|DestroyActor))))

(fn CanAttachToActor
  (params (Actor :type Actor%20Object%20Reference))
  (return (or (CallFunction|IsAttachableActorClass Actor) (CallFunction|HasAttachableTags Actor))))

(fn CanDeployOnActor
  (params (Actor :type Actor%20Object%20Reference) (Component :type Primitive%20Component%20Object%20Reference))
  (bind _returnvalue (GameplayTags|HasAnyMatchingGameplayTags(Message) Actor (Variables|Default|GetDisallowedActorTags)))
  (return (and (not _returnvalue) (Utilities|Enum|NotEqual(Enum) (Collision|GetCollisionObjectType Component) "ECC_GameTraceChannel18"))))

(fn ClampNormal
  (params (Normal :type Vector) (HitActor :type Actor%20Object%20Reference))
  (bind _returnvalue (GameplayTags|HasMatchingGameplayTag(Message) HitActor "(TagName=\"Vehicle\")"))
  (if _returnvalue
    (return "0.000000,0.000000,1.000000")
    (else
      (return (Variables|GetNormal)))))

(fn GetEnergy
  (bind _returnvalue (CallFunction|GetOwnerWeapon))
  (bind _asb_athena_shield_weapon (Utilities|Casting|CastToB_Athena_ShieldWeapon _returnvalue)
    (:then
      (return (Variables|/ShielderGameplay/Gameplay/B_Shield_EnergyTracker_Component.B_Shield_EnergyTracker_Component_C|GetCurrentEnergy (Variables|/ShielderGameplay/Gameplay/BGA_Shielder_Parent.BGA_Shielder_Parent_C|GetShieldEnergyComponent (Variables|/ShielderGameplay/Gameplay/B_Athena_ShieldWeapon.B_Athena_ShieldWeapon_C|GetShieldBGA _asb_athena_shield_weapon)))))
    (:CastFailed
      (return 0.0))))

(fn GetSpawnTransform
  (params (Hit :type Hit%20Result%20Structure))
  (bind (_bblockinghit _binitialoverlap _time _distance _location _impactpoint _normal) (Collision|BreakHitResult Hit))
  (bind (_returnvalue_roll _returnvalue_pitch _returnvalue_yaw) (Math|Rotator|BreakRotator (Math|Conversions|RotationFromXVector _normal)))
  (bind _returnvalue (* _normal 2e+01))
  (bind _returnvalue_2 (+ _location _returnvalue))
  (Rendering|Debug|DrawDebugLine _location _returnvalue_2 "(R=0.026373,G=1.000000,B=0.000000,A=1.000000)" 1e+01 2.0)
  (return (Math|Transform|MakeTransform _returnvalue_2 (Math|Rotator|MakeRotator _returnvalue_roll (+ _returnvalue_pitch 9e+01) (+ _returnvalue_yaw 1.8e+02)))))

(fn GiveItemEntryToDeployedShield
  (params (DeployedShield :type BGA%20Shielder%20Deployed%20Object%20Reference))
  (Class|FortItemEntryComponent|SetOwnedItem (Variables|/ShielderGameplay/Gameplay/BGA_Shielder_Deployed.BGA_Shielder_Deployed_C|GetFortItemEntry DeployedShield) (Class|FortItemEntryComponent|GetOwnedItem (Variables|Default|GetFortItemEntry))))

(fn HasAttachableTags
  (params (Actor :type Actor%20Object%20Reference))
  (bind _returnvalue (GameplayTags|HasAnyMatchingGameplayTags(Message) Actor (Variables|Default|GetAttachableActorTags)))
  (return _returnvalue))

(fn IsAttachableActorClass
  (params (Actor :type Actor%20Object%20Reference))
  (for _array_element (Variables|Default|GetAttachableActorClasses)
    (if (Utilities|ClassIsChildOf (Utilities|GetClass (Variables|GetActor)) _array_element)
      (return true)))
  (return false))

(fn IsValidHit
  (params (Hit :type Hit%20Result%20Structure))
  (return false))

(fn MatchRotationForVehicles
  (params (Actor :type Actor%20Object%20Reference) (Shield :type Actor%20Object%20Reference))
  (bind _returnvalue (GameplayTags|HasMatchingGameplayTag(Message) Actor "(TagName=\"Vehicle\")"))
  (if _returnvalue
    (Transformation|SetActorRotation (Variables|GetShield) (Transformation|GetActorRotation Actor) false)))

(fn RemoveShieldItem
  (bind _returnvalue (Game|GetInstigator))
  (bind _shieldwid (Variables|Default|GetShieldWID))
  (bind _returnvalue_2 (Fortnite|GetFortPlayerControllerfromActor _returnvalue))
  (bind _returnvalue_3 (Inventory|FindExistingItemForDefinition :self _returnvalue_2 :ItemDefinition _shieldwid :bInStorageVault false))
  (bind _returnvalue_4 (Item|GetItemGuid _returnvalue_3))
  (Utilities|Casting|CastToFortPlayerPawn _returnvalue)
  (Utilities|Casting|CastToFortWorldItem _returnvalue_3)
  (Class|FortItemEntryComponent|TakeItemfromInventoryOwner (Variables|Default|GetFortItemEntry) _returnvalue_4 -1 _returnvalue_2 true)
  (Inventory|RemoveItemFromPlayerByGuid _returnvalue_2 _returnvalue_4 1 true))

(fn ShouldBounceOff
  (params (Actor :type Actor%20Object%20Reference))
  (Utilities|Casting|CastToBGA_Shielder_Deployed Actor
    (:then
      (return true))
    (:CastFailed
      (return false))))

(fn SpawnItemPickup
  (Inventory|SpawnPickupInWorldWithClassAndItemEntry (Class|FortItemEntryComponent|GetOwnedItem (Variables|Default|GetFortItemEntry)) "/Script/FortniteGame.FortPickupAthena" (Transformation|GetActorLocation) "0.000000,0.000000,1.000000" 0 true true false "Tossed" "Unset" 0 false)
  (Actor|DestroyActor))

(fn SpawnShield
  (params (Transform :type Transform%20%28by%20ref%29))
  (bind _returnvalue (Building|SpawnBuildingGameplayActor (Variables|Default|GetShieldClass) Transform (Game|GetInstigator) true))
  (return _returnvalue))

(fn TransferEnergy
  (params (DeployedShield :type BGA%20Shielder%20Deployed%20Object%20Reference))
  (bind _deployedshield (Variables|GetDeployedShield))
  (bind _shieldenergycomponent (Variables|/ShielderGameplay/Gameplay/BGA_Shielder_Parent.BGA_Shielder_Parent_C|GetShieldEnergyComponent _deployedshield))
  (bind _returnvalue (CallFunction|GetOwnerWeapon))
  (bind _asb_athena_shield_weapon (Utilities|Casting|CastToB_Athena_ShieldWeapon _returnvalue))
  (Utilities|Time|ClearandInvalidateTimerbyHandle (Variables|/ShielderGameplay/Gameplay/B_Shield_EnergyTracker_Component.B_Shield_EnergyTracker_Component_C|GetRechargeDelayTimer (Variables|/ShielderGameplay/Gameplay/BGA_Shielder_Parent.BGA_Shielder_Parent_C|GetShieldEnergyComponent (Variables|/ShielderGameplay/Gameplay/B_Athena_ShieldWeapon.B_Athena_ShieldWeapon_C|GetShieldBGA _asb_athena_shield_weapon))))
  (Default|StopEnergyTimer (Variables|/ShielderGameplay/Gameplay/BGA_Shielder_Parent.BGA_Shielder_Parent_C|GetShieldEnergyComponent (Variables|/ShielderGameplay/Gameplay/B_Athena_ShieldWeapon.B_Athena_ShieldWeapon_C|GetShieldBGA _asb_athena_shield_weapon)))
  (Class|BShieldEnergyTrackerComponent|StopDepleting (Variables|/ShielderGameplay/Gameplay/BGA_Shielder_Parent.BGA_Shielder_Parent_C|GetShieldEnergyComponent (Variables|/ShielderGameplay/Gameplay/B_Athena_ShieldWeapon.B_Athena_ShieldWeapon_C|GetShieldBGA _asb_athena_shield_weapon)) false true)
  (Class|BShieldEnergyTrackerComponent|Initialize _shieldenergycomponent _asb_athena_shield_weapon false)
  (Class|BShieldEnergyTrackerComponent|StartDepleting _shieldenergycomponent)
  (Actor|DestroyActor (Variables|/ShielderGameplay/Gameplay/B_Athena_ShieldWeapon.B_Athena_ShieldWeapon_C|GetShieldBGA _asb_athena_shield_weapon)))

(fn TryAttach
  (params (HitActor :type Actor%20Object%20Reference) (SpawnedActor :type Actor%20Object%20Reference))
  (bind _canattach (CallFunction|CanAttachtoActor HitActor))
  (if _canattach
    (Transformation|AttachActorToComponent SpawnedActor (Variables|/Script/Engine.Actor|GetRootComponent HitActor) "None" "KeepWorld" "KeepWorld" "KeepWorld" true)))

(fn ConstructionScript
  (|CallParentFunction))
