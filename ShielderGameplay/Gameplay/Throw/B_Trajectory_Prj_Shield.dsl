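; Trajectory preview Blueprint for a thrown shield, orienting its target marker whenever the trajectory spline updates.
; GetSurfaceNormal sphere-traces through the Target along the line from the instigator, ignoring the spline endpoint passed in.
; On a hit, the Target aligns to the surface normal, using the instigator's projected forward vector turned a quarter turn.
; Otherwise it keeps pitch and roll and takes yaw from the pawn's targeting rotation plus YawOffset.
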(event EventBeginPlay
  (|CallParentFunction) ; This node is disabled and will not be called.\nDrag off pins to build functionality.
)

(event Collision|EventActorBeginOverlap
  (params (OtherActor :type Actor%20Object%20Reference))
  (|CallParentFunction OtherActor) ; This node is disabled and will not be called.\nDrag off pins to build functionality.
)

(event EventTick
  (params (DeltaSeconds :type Float%20%28single-precision%29))
  (|CallParentFunction DeltaSeconds) ; This node is disabled and will not be called.\nDrag off pins to build functionality.
)

(event Trajectory|EventSetTrajectorySpline
  (params (SplinePoints :type Array%20of%20Vectors) (SplineTangents :type Array%20of%20Vectors))
  (|CallParentFunction SplinePoints SplineTangents)
  (Utilities|FlowControl|Sequence
    (:then_0
      (bind (_bblockinghit _normal) (CallFunction|GetSurfaceNormal (Utilities|Array|Get(acopy) SplinePoints (- (Utilities|Array|Length SplinePoints) 1))))
      (if _bblockinghit
        (CallFunction|UpdateRotationforSurface _normal)
        (else
          (bind _asfort_player_pawn (Utilities|Casting|CastToFortPlayerPawn (Game|GetInstigator)))
          (Transformation|SetWorldRotation (Variables|Default|GetTarget) (Math|Rotator|MakeRotator (.roll (Transformation|GetWorldRotation (Variables|Default|GetTarget))) (.pitch (Transformation|GetWorldRotation (Variables|Default|GetTarget))) (+ (.yaw (Targeting|GetDefaultTargetingRotation _asfort_player_pawn)) (Variables|Default|GetYawOffset))) false false))))))

(fn GetSurfaceNormal
  (params (Position :type Vector))
  (bind _outhit (Collision|SphereTraceByProfile (+ (Transformation|GetWorldLocation (Variables|Default|GetTarget)) (* (Math|Vector|NegateVector (Math|Vector|GetUnitDirection(Vector) (Transformation|GetActorLocation (Game|GetInstigator)) (Transformation|GetWorldLocation (Variables|Default|GetTarget)))) 1e+02)) (+ (Transformation|GetWorldLocation (Variables|Default|GetTarget)) (* (Math|Vector|GetUnitDirection(Vector) (Transformation|GetActorLocation (Game|GetInstigator)) (Transformation|GetWorldLocation (Variables|Default|GetTarget))) 1e+02)) 2e+01 "FortProjectileOverlapPawns"))
  (return (Collision|BreakHitResult _outhit) (.ImpactNormal (Collision|BreakHitResult _outhit))))

(fn UpdateRotationForSurface
  (params (Normal :type Vector%20%28by%20ref%29))
  (Transformation|SetWorldRotation (Variables|Default|GetTarget) (Math|Rotator|MakeRotfromXZ (Math|Vector|RotateVectorAroundAxis (Math|Vector|ProjectVectorontoPlane (Transformation|GetActorForwardVector (Game|GetInstigator)) Normal) -9e+01 Normal) Normal) false false))

(fn ConstructionScript
  (|CallParentFunction))
